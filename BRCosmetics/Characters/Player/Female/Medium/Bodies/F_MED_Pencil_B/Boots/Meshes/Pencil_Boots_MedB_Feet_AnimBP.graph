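# Post-process-style AnimBP for the feet of the female medium Pencil_B boots, adding secondary motion to boots, knots, laces and loops.
# It copies the attached parent's pose, then chains weighted calf/foot constraints and AnimDynamics on the boot bones.
# Jump and fall events toggle `jumping`, which eases in additive knot rotations before per-bone lace and loop AnimDynamics.
# The DeformRig Control Rig and Input Pose nodes are unconnected.
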AnimBlueprint /BRCosmetics/Characters/Player/Female/Medium/Bodies/F_MED_Pencil_B/Boots/Meshes/Pencil_Boots_MedB_Feet_AnimBP
target_skeleton=&/Pencil_Boots_MedB_Feet_Skeleton
graphs [
 AnimGraph {nodes [
   Control Rig() {
    alphaBoolBlend {blendOption=Linear}
    controlRigAssetReference {blueprintRigClass=&/Pencil_Boots_MedB_Feet_DeformRig.Pencil_Boots_MedB_Feet_DeformRig_C}
   }
   Input Pose()
   n1 = "Constraint - dyn_boot_l_MSTR (PPP)"!(
    ComponentPose=Local To Component(LocalPose=Copy Pose From Mesh() {bCopyCurves=true, bUseAttachedParent=true})
    Alpha=1
    ConstraintWeights_0=0.7301
    ConstraintWeights_1:float=0.0823
    ConstraintWeights_2:float=0.1876
   ) -> (Pose) {
    alphaBoolBlend {blendOption=Linear}
    boneToModify {boneName=dyn_boot_l_MSTR}
    constraintSetup [{targetBone {boneName=calf_twist_02_l}}, {targetBone {boneName=calf_l}}, {targetBone {boneName=calf_twist_01_l}}]
    constraintWeights [1,1,1]
   }
   n2 = "Constraint - dyn_boot_r_MSTR (PPP)"!(ComponentPose=@n1, Alpha=1, ConstraintWeights_0=0.7301, ConstraintWeights_1:float=0.0823, ConstraintWeights_2:float=0.1876) -> (Pose) {
    alphaBoolBlend {blendOption=Linear}
    boneToModify {boneName=dyn_boot_r_MSTR}
    constraintSetup [{targetBone {boneName=calf_twist_02_r}}, {targetBone {boneName=calf_r}}, {targetBone {boneName=calf_twist_01_r}}]
    constraintWeights [1,1,1]
   }
   n3 = "Constraint - dyn_boot_l_Knot (PPPP)"!(
    ComponentPose=@n2
    Alpha=1
    ConstraintWeights_0=0.5097
    ConstraintWeights_1:float=0.009
    ConstraintWeights_2:float=0.0821
    ConstraintWeights_3:float=0.3992
   ) -> (Pose) {
    alphaBoolBlend {blendOption=Linear}
    boneToModify {boneName=dyn_boot_l_Knot}
    constraintSetup [
     {targetBone {boneName=calf_twist_02_l}}
     {targetBone {boneName=foot_l}}
     {targetBone {boneName=calf_l}}
     {targetBone {boneName=calf_twist_01_l}}
    ]
    constraintWeights [1,1,1,1]
   }
   n4 = "Constraint - dyn_boot_r_Knot (PPPP)"!(
    ComponentPose=@n3
    Alpha=1
    ConstraintWeights_0=0.009
    ConstraintWeights_1:float=0.5097
    ConstraintWeights_2:float=0.0821
    ConstraintWeights_3:float=0.3992
   ) -> (Pose) {
    alphaBoolBlend {blendOption=Linear}
    boneToModify {boneName=dyn_boot_r_Knot}
    constraintSetup [
     {targetBone {boneName=foot_r}}
     {targetBone {boneName=calf_twist_02_r}}
     {targetBone {boneName=calf_r}}
     {targetBone {boneName=calf_twist_01_r}}
    ]
    constraintWeights [1,1,1,1]
   }
   n5 = Anim Dynamics - Bone: dyn_boot_l!(ComponentPose=@n4, Alpha=0.5, GravityOverride:Vector="0,0,0", ExternalForce="0,0,0") -> (Pose) {
    angularDampingOverride=0.9999
    angularSpringConstant=1500
    bAngularSpring=true
    bGravityOverrideInSimSpace=true
    bOverrideAngularDamping=true
    gravityScale=0.8
    simulationSpace=BoneRelative
    alphaBoolBlend {blendOption=Linear}
    boundBone {boneName=dyn_boot_l}
    componentAppliedLinearAccClamp=<5000 5000 3500>
    componentLinearAccScale=<1.5 1.5 1.25>
    componentLinearVelScale=<1.5 1.5 1.25>
    physicsBodyDefinitions [{
      sphereCollisionRadius=0
      boundBone {boneName=dyn_boot_l}
      boxExtents=<10 20 20>
      constraintSetup {angularLimitsMax=<15 15 35>, angularLimitsMin=<-15 -15 -35>, angularTarget=<0 0 0>}
      localJointOffset=<1 0 0>
     }]
    relativeSpaceBone {boneName=root}
   }
   n6 = Anim Dynamics - Bone: dyn_boot_r!(ComponentPose=@n5, Alpha=0.5, GravityOverride:Vector="0,0,0", ExternalForce="0,0,0") -> (Pose) {
    angularDampingOverride=0.9999
    angularSpringConstant=1500
    bAngularSpring=true
    bGravityOverrideInSimSpace=true
    bOverrideAngularDamping=true
    gravityScale=0.8
    simulationSpace=BoneRelative
    alphaBoolBlend {blendOption=Linear}
    boundBone {boneName=dyn_boot_r}
    componentAppliedLinearAccClamp=<5000 5000 3500>
    componentLinearAccScale=<1.5 1.5 1.25>
    componentLinearVelScale=<1.5 1.5 1.25>
    physicsBodyDefinitions [{
      sphereCollisionRadius=0
      boundBone {boneName=dyn_boot_r}
      boxExtents=<10 20 20>
      constraintSetup {angularLimitsMax=<15 15 35>, angularLimitsMin=<-15 -15 -35>, angularTarget=<0 0 0>}
      localJointOffset=<1 0 0>
     }]
    relativeSpaceBone {boneName=root}
   }
   n7 = Get jumping() -> (jumping:double)
   n8 = "Transform (Modify) Bone - Bone: dyn_boot_l_Knot"(ComponentPose=@n6, Alpha=@n7, Translation="0,0,0", Rotation="0,0,-70", Scale="1,1,1") {
    rotationMode=BMM_Additive
    rotationSpace=BCS_BoneSpace
    alphaBoolBlend {blendOption=Linear}
    alphaScaleBiasClamp {bInterpResult=true, interpSpeedIncreasing=2}
    boneToModify {boneName=dyn_boot_l_Knot}
    rotation {roll=-70}
   }
   n9 = "Transform (Modify) Bone - Bone: dyn_boot_r_Knot"(ComponentPose=@n8, Alpha=@n7, Translation="0,0,0", Rotation="70,0,0", Scale="1,1,1") {
    rotationMode=BMM_Additive
    rotationSpace=BCS_BoneSpace
    alphaBoolBlend {blendOption=Linear}
    alphaScaleBiasClamp {bInterpResult=true, interpSpeedIncreasing=2}
    boneToModify {boneName=dyn_boot_r_Knot}
    rotation {pitch=70}
   }
   n10 = Anim Dynamics - Bone: dyn_boot_l_lace_a_01(ComponentPose=@n9, Alpha=1, ExternalForce="0,0,0") {
    angularDampingOverride=0.98
    angularSpringConstant=75
    bGravityOverrideInSimSpace=true
    bOverrideAngularDamping=true
    simulationSpace=RootRelative
    alphaBoolBlend {blendOption=Linear}
    boundBone {boneName=dyn_boot_l_lace_a_01}
    componentAppliedLinearAccClamp=<1500 1500 1500>
    componentLinearAccScale=<0.7 0.7 0.7>
    componentLinearVelScale=<0.7 0.7 0.7>
    physicsBodyDefinitions [{
      sphereCollisionRadius=0
      boundBone {boneName=dyn_boot_l_lace_a_01}
      boxExtents=<10 10 10>
      constraintSetup {angularLimitsMax=<5 2 30>, angularLimitsMin=<-5 -15 -30>, angularTarget=<1 0 0>}
      localJointOffset=<0.5 0 0>
     }]
    relativeSpaceBone {boneName=foot_l}
   }
   n11 = Anim Dynamics - Bone: dyn_boot_l_lace_a_02(ComponentPose=@n10, Alpha=1, ExternalForce="0,0,0") {
    angularDampingOverride=0.98
    angularSpringConstant=75
    bGravityOverrideInSimSpace=true
    bOverrideAngularDamping=true
    simulationSpace=RootRelative
    alphaBoolBlend {blendOption=Linear}
    boundBone {boneName=dyn_boot_l_lace_a_02}
    componentAppliedLinearAccClamp=<1500 1500 1500>
    componentLinearAccScale=<0.7 0.7 0.7>
    componentLinearVelScale=<0.7 0.7 0.7>
    physicsBodyDefinitions [{
      sphereCollisionRadius=0
      boundBone {boneName=dyn_boot_l_lace_a_02}
      boxExtents=<10 10 10>
      constraintSetup {angularLimitsMax=<5 2 30>, angularLimitsMin=<-5 -15 -30>, angularTarget=<1 0 0>}
      localJointOffset=<0.5 0 0>
     }]
    relativeSpaceBone {boneName=foot_l}
   }
   n12 = Anim Dynamics - Bone: dyn_boot_l_lace_b_01(ComponentPose=@n11, Alpha=1, ExternalForce="0,0,0") {
    angularDampingOverride=0.98
    angularSpringConstant=50
    bGravityOverrideInSimSpace=true
    bOverrideAngularDamping=true
    simulationSpace=RootRelative
    alphaBoolBlend {blendOption=Linear}
    boundBone {boneName=dyn_boot_l_lace_b_01}
    componentAppliedLinearAccClamp=<1500 1500 1500>
    componentLinearAccScale=<0.8 0.8 0.8>
    componentLinearVelScale=<0.8 0.8 0.8>
    physicsBodyDefinitions [{
      sphereCollisionRadius=0
      boundBone {boneName=dyn_boot_l_lace_b_01}
      boxExtents=<12 12 12>
      constraintSetup {angularLimitsMax=<5 2 30>, angularLimitsMin=<-5 -15 -30>, angularTarget=<1 0 0>}
      localJointOffset=<1 0 0>
     }]
    relativeSpaceBone {boneName=foot_l}
   }
   n13 = Anim Dynamics - Bone: dyn_boot_l_lace_b_02(ComponentPose=@n12, Alpha=1, ExternalForce="0,0,0") {
    angularDampingOverride=0.98
    angularSpringConstant=50
    bGravityOverrideInSimSpace=true
    bOverrideAngularDamping=true
    simulationSpace=RootRelative
    alphaBoolBlend {blendOption=Linear}
    boundBone {boneName=dyn_boot_l_lace_b_02}
    componentAppliedLinearAccClamp=<1500 1500 1500>
    componentLinearAccScale=<0.8 0.8 0.8>
    componentLinearVelScale=<0.8 0.8 0.8>
    physicsBodyDefinitions [{
      sphereCollisionRadius=0
      boundBone {boneName=dyn_boot_l_lace_b_02}
      boxExtents=<12 12 12>
      constraintSetup {angularLimitsMax=<5 2 30>, angularLimitsMin=<-5 -15 -30>, angularTarget=<1 0 0>}
      localJointOffset=<1 0 0>
     }]
    relativeSpaceBone {boneName=foot_l}
   }
   n14 = Anim Dynamics - Bone: dyn_boot_l_loop_a(ComponentPose=@n13, Alpha=0.8, ExternalForce="0,0,0") {
    angularDampingOverride=0.97
    angularSpringConstant=300
    bAngularSpring=true
    bGravityOverrideInSimSpace=true
    bOverrideAngularDamping=true
    simulationSpace=RootRelative
    alphaBoolBlend {blendOption=Linear}
    boundBone {boneName=dyn_boot_l_loop_a}
    componentAppliedLinearAccClamp=<3500 3500 3500>
    componentLinearAccScale=<0.85 0.85 0.85>
    componentLinearVelScale=<0.85 0.85 0.85>
    physicsBodyDefinitions [{
      sphereCollisionRadius=0
      boundBone {boneName=dyn_boot_l_loop_a}
      boxExtents=<20 15 15>
      constraintSetup {angularLimitsMax=<5 2 15>, angularLimitsMin=<-5 -15 -15>, angularTarget=<1 0 0>}
      localJointOffset=<1 0 0>
     }]
    relativeSpaceBone {boneName=foot_l}
   }
   n15 = Anim Dynamics - Bone: dyn_boot_l_loop_b(ComponentPose=@n14, Alpha=0.8, ExternalForce="0,0,0") {
    angularDampingOverride=0.97
    angularSpringConstant=300
    bAngularSpring=true
    bGravityOverrideInSimSpace=true
    bOverrideAngularDamping=true
    simulationSpace=RootRelative
    alphaBoolBlend {blendOption=Linear}
    boundBone {boneName=dyn_boot_l_loop_b}
    componentAppliedLinearAccClamp=<3500 3500 3500>
    componentLinearAccScale=<0.85 0.85 0.85>
    componentLinearVelScale=<0.85 0.85 0.85>
    physicsBodyDefinitions [{
      sphereCollisionRadius=0
      boundBone {boneName=dyn_boot_l_loop_b}
      boxExtents=<20 15 15>
      constraintSetup {angularLimitsMax=<5 2 15>, angularLimitsMin=<-5 -15 -15>, angularTarget=<1 0 0>}
      localJointOffset=<1 0 0>
     }]
    relativeSpaceBone {boneName=foot_l}
   }
   n16 = Anim Dynamics - Bone: dyn_boot_r_lace_a_01(ComponentPose=@n15, Alpha=1, ExternalForce="0,0,0") {
    angularDampingOverride=0.98
    angularSpringConstant=75
    bGravityOverrideInSimSpace=true
    bOverrideAngularDamping=true
    simulationSpace=RootRelative
    alphaBoolBlend {blendOption=Linear}
    boundBone {boneName=dyn_boot_r_lace_a_01}
    componentAppliedLinearAccClamp=<1500 1500 1500>
    componentLinearAccScale=<0.8 0.8 0.8>
    componentLinearVelScale=<0.8 0.8 0.8>
    physicsBodyDefinitions [{
      sphereCollisionRadius=0
      boundBone {boneName=dyn_boot_r_lace_a_01}
      boxExtents=<10 10 10>
      constraintSetup {angularLimitsMax=<5 2 30>, angularLimitsMin=<-5 -15 -30>, angularTarget=<1 0 0>}
      localJointOffset=<0.75 0 0>
     }]
    relativeSpaceBone {boneName=foot_r}
   }
   n17 = Anim Dynamics - Bone: dyn_boot_r_lace_a_02(ComponentPose=@n16, Alpha=1, ExternalForce="0,0,0") {
    angularDampingOverride=0.98
    angularSpringConstant=75
    bGravityOverrideInSimSpace=true
    bOverrideAngularDamping=true
    simulationSpace=RootRelative
    alphaBoolBlend {blendOption=Linear}
    boundBone {boneName=dyn_boot_r_lace_a_02}
    componentAppliedLinearAccClamp=<1500 1500 1500>
    componentLinearAccScale=<0.8 0.8 0.8>
    componentLinearVelScale=<0.8 0.8 0.8>
    physicsBodyDefinitions [{
      sphereCollisionRadius=0
      boundBone {boneName=dyn_boot_r_lace_a_02}
      boxExtents=<10 10 10>
      constraintSetup {angularLimitsMax=<5 2 30>, angularLimitsMin=<-5 -15 -30>, angularTarget=<1 0 0>}
      localJointOffset=<0.75 0 0>
     }]
    relativeSpaceBone {boneName=foot_r}
   }
   n18 = Anim Dynamics - Bone: dyn_boot_r_lace_b_01(ComponentPose=@n17, Alpha=1, ExternalForce="0,0,0") {
    angularDampingOverride=0.98
    angularSpringConstant=50
    bGravityOverrideInSimSpace=true
    bOverrideAngularDamping=true
    simulationSpace=RootRelative
    alphaBoolBlend {blendOption=Linear}
    boundBone {boneName=dyn_boot_r_lace_b_01}
    componentAppliedLinearAccClamp=<1500 1500 1500>
    componentLinearAccScale=<0.7 0.7 0.7>
    componentLinearVelScale=<0.7 0.7 0.7>
    physicsBodyDefinitions [{
      sphereCollisionRadius=0
      boundBone {boneName=dyn_boot_r_lace_b_01}
      boxExtents=<12 12 12>
      constraintSetup {angularLimitsMax=<5 2 30>, angularLimitsMin=<-5 -15 -30>, angularTarget=<1 0 0>}
      localJointOffset=<1 0 0>
     }]
    relativeSpaceBone {boneName=foot_r}
   }
   n19 = Anim Dynamics - Bone: dyn_boot_r_lace_b_02(ComponentPose=@n18, Alpha=1, ExternalForce="0,0,0") {
    angularDampingOverride=0.98
    angularSpringConstant=50
    bGravityOverrideInSimSpace=true
    bOverrideAngularDamping=true
    simulationSpace=RootRelative
    alphaBoolBlend {blendOption=Linear}
    boundBone {boneName=dyn_boot_r_lace_b_02}
    componentAppliedLinearAccClamp=<1500 1500 1500>
    componentLinearAccScale=<0.7 0.7 0.7>
    componentLinearVelScale=<0.7 0.7 0.7>
    physicsBodyDefinitions [{
      sphereCollisionRadius=0
      boundBone {boneName=dyn_boot_r_lace_b_02}
      boxExtents=<12 12 12>
      constraintSetup {angularLimitsMax=<5 2 30>, angularLimitsMin=<-5 -15 -30>, angularTarget=<1 0 0>}
      localJointOffset=<1 0 0>
     }]
    relativeSpaceBone {boneName=foot_r}
   }
   n20 = Anim Dynamics - Bone: dyn_boot_r_loop_a(ComponentPose=@n19, Alpha=0.8, ExternalForce="0,0,0") {
    angularDampingOverride=0.97
    angularSpringConstant=300
    bAngularSpring=true
    bGravityOverrideInSimSpace=true
    bOverrideAngularDamping=true
    simulationSpace=RootRelative
    alphaBoolBlend {blendOption=Linear}
    boundBone {boneName=dyn_boot_r_loop_a}
    componentAppliedLinearAccClamp=<3500 3500 3500>
    componentLinearAccScale=<0.85 0.85 0.85>
    componentLinearVelScale=<0.85 0.85 0.85>
    physicsBodyDefinitions [{
      sphereCollisionRadius=0
      boundBone {boneName=dyn_boot_r_loop_a}
      boxExtents=<20 15 15>
      constraintSetup {angularLimitsMax=<5 2 15>, angularLimitsMin=<-5 -15 -15>, angularTarget=<1 0 0>}
      localJointOffset=<1 0 0>
     }]
    relativeSpaceBone {boneName=foot_r}
   }
   n21 = Anim Dynamics - Bone: dyn_boot_r_loop_b(ComponentPose=@n20, Alpha=0.8, ExternalForce="0,0,0") {
    angularDampingOverride=0.97
    angularSpringConstant=300
    bAngularSpring=true
    bGravityOverrideInSimSpace=true
    bOverrideAngularDamping=true
    simulationSpace=RootRelative
    alphaBoolBlend {blendOption=Linear}
    boundBone {boneName=dyn_boot_r_loop_b}
    componentAppliedLinearAccClamp=<3500 3500 3500>
    componentLinearAccScale=<0.85 0.85 0.85>
    componentLinearVelScale=<0.85 0.85 0.85>
    physicsBodyDefinitions [{
      sphereCollisionRadius=0
      boundBone {boneName=dyn_boot_r_loop_b}
      boxExtents=<20 15 15>
      constraintSetup {angularLimitsMax=<5 2 15>, angularLimitsMin=<-5 -15 -15>, angularTarget=<1 0 0>}
      localJointOffset=<1 0 0>
     }]
    relativeSpaceBone {boneName=foot_r}
   }
   Output Pose(Result=Component To Local(ComponentPose=@n21)) {layerGroup=None}
  ]}
 EventGraph {nodes [
   n1 = Event OnBeginFalling()
   Set jumping(execute=[@n1.then, @n3.then], jumping:double=1) -> (Output_Get:double)
   n2 = Event OnEndFalling()
   Set jumping(execute=[@n2.then, @n4.then], jumping=0)
   n3 = Event OnBeginJumping()
   n4 = Event OnEndJumping()
   Event BlueprintUpdateAnimation()
  ]}
]
variables [jumping {type=number}]
